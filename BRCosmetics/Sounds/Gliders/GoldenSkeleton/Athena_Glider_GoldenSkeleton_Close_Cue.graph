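# Sound cue for the Golden Skeleton glider's close sound in Athena (Battle Royale).
# A single wave player plays the Glider_GoldenSkeleton_Close sound wave directly as the root, with no randomization or modulation nodes.
# Shared glider open/close attenuation and concurrency assets, the cosmetic glider-close property sheet and the glider landing sound class govern spatialization, voice limiting and mix, at a reduced volume.
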SoundCue /BRCosmetics/Sounds/Gliders/GoldenSkeleton/Athena_Glider_GoldenSkeleton_Close_Cue
root=WavePlayer_1
nodes [{id=WavePlayer_1, properties {soundWaveAssetPtr=&/Glider_GoldenSkeleton_Close}}]
properties {
 attenuationSettings=/Game/Sounds/Attenuation/Fort_Player_Glider_OpenClose_Athena_Att
 audioPropertiesSheet=/BRCosmetics/Sounds/Settings/PropertySheets/Glider/APS_Cosmetic_Glider_Close
 maxDistance=10000
 soundClassObject=&Fort_Gliders_Land
 volumeMultiplier=0.6
 concurrencySet [&Athena_GliderOpenClose_Conc]
}
