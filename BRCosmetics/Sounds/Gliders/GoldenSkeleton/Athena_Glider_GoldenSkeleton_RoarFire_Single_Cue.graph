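# Sound cue for a single roar-fire burst on the Golden Skeleton glider cosmetic.
# It plays one fire burst wave through a modulator that randomizes pitch, at reduced volume with glider thrust attenuation and the glider landing sound class.
# A concurrency override caps it at two instances and prevents new ones beyond that, alongside an assigned glider open/close concurrency set.
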SoundCue /BRCosmetics/Sounds/Gliders/GoldenSkeleton/Athena_Glider_GoldenSkeleton_RoarFire_Single_Cue
root=Modulator_0
nodes [
 {id=Modulator_0, children [WavePlayer_3], properties {pitchMax=1.1, pitchMin=0.8}}
 {id=WavePlayer_3, properties {soundWaveAssetPtr=&/Glider_GoldenSkeleton_FireBurst_02}}
]
properties {
 attenuationSettings=&Fort_Player_Glider_Thrust_Athena_Att
 bOverrideConcurrency=true
 maxDistance=12500
 soundClassObject=&Fort_Gliders_Land
 volumeMultiplier=0.35
 concurrencyOverrides {maxCount=2, resolutionRule=PreventNew}
 concurrencySet [&Athena_GliderOpenClose_Conc]
}
