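# Sound cue for the roar of the Golden Skeleton glider cosmetic in Athena.
# A random node picks between two equally weighted roar waves and feeds a modulator that varies pitch on each play.
# It plays at reduced volume through glider thrust attenuation and the glider landing sound class, with a concurrency override that prevents new instances beyond a small cap.
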SoundCue /BRCosmetics/Sounds/Gliders/GoldenSkeleton/Athena_Glider_GoldenSkeleton_Roar_Cue
root=Modulator_0
nodes [
 {id=Modulator_0, children [Random_0], properties {pitchMax=1.1, pitchMin=0.8}}
 {id=Random_0, children [WavePlayer_0,WavePlayer_2], properties {weights [1,1]}}
 {id=WavePlayer_0, properties {soundWaveAssetPtr=&/Glider_GoldenSkeleton_Roar_01}}
 {id=WavePlayer_2, properties {soundWaveAssetPtr=&/Glider_GoldenSkeleton_Roar_02}}
]
properties {
 attenuationSettings=&Fort_Player_Glider_Thrust_Athena_Att
 bOverrideConcurrency=true
 maxDistance=12500
 soundClassObject=&Fort_Gliders_Land
 volumeMultiplier=0.35
 concurrencyOverrides {maxCount=2, resolutionRule=PreventNew}
 concurrencySet [&Athena_GliderOpenClose_Conc]
}
