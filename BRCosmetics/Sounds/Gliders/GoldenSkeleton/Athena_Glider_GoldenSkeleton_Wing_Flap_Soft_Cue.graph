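# Sound cue for the soft wing flap of the Golden Skeleton glider.
# One of four equally weighted soft wing samples is picked at random and passed through a modulator for slight pitch variation.
# It plays at reduced volume with glider thrust attenuation and the glider landing sound class.
# A concurrency override caps simultaneous instances, taking precedence over the assigned glider open/close concurrency set.
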SoundCue /BRCosmetics/Sounds/Gliders/GoldenSkeleton/Athena_Glider_GoldenSkeleton_Wing_Flap_Soft_Cue
root=Modulator_0
nodes [
 {id=Modulator_0, children [Random_0], properties {pitchMax=1.1, pitchMin=0.9}}
 {id=Random_0, children [WavePlayer_1,WavePlayer_2,WavePlayer_4,WavePlayer_5], properties {weights [1,1,1,1]}}
 {id=WavePlayer_1, properties {soundWaveAssetPtr=&/Glider_GoldenSkeleton_Wings_Soft_01}}
 {id=WavePlayer_2, properties {soundWaveAssetPtr=&/Glider_GoldenSkeleton_Wings_Soft_02}}
 {id=WavePlayer_4, properties {soundWaveAssetPtr=&/Glider_GoldenSkeleton_Wings_Soft_03}}
 {id=WavePlayer_5, properties {soundWaveAssetPtr=&/Glider_GoldenSkeleton_Wings_Soft_04}}
]
properties {
 attenuationSettings=&Fort_Player_Glider_Thrust_Athena_Att
 bOverrideConcurrency=true
 maxDistance=12500
 soundClassObject=&Fort_Gliders_Land
 volumeMultiplier=0.25
 concurrencyOverrides {maxCount=3}
 concurrencySet [&Athena_GliderOpenClose_Conc]
}
