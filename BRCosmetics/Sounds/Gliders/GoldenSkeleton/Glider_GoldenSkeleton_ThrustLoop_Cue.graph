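# Sound cue for the looping thrust audio of the Golden Skeleton cosmetic glider.
# A single looping wave player feeds a modulator that randomizes pitch, with boosted cue volume.
# It uses glider thrust attenuation and keeps playing when silent through virtualization.
# Overridden concurrency limits instances per owner and stops the oldest, taking the place of the assigned open/close concurrency set.
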SoundCue /BRCosmetics/Sounds/Gliders/GoldenSkeleton/Glider_GoldenSkeleton_ThrustLoop_Cue
root=Modulator_0
nodes [
 {id=Modulator_0, children [WavePlayer_1], properties {pitchMax=1.1, pitchMin=0.85, volumeMin=1}}
 {id=WavePlayer_1, properties {bLooping=true, soundWaveAssetPtr=&/Glider_GoldenSkeleton_ThrustLoop}}
]
properties {
 attenuationSettings=&Fort_Player_Glider_Thrust_Athena_Att
 audioPropertiesSheet=&APS_Cosmetic_Glider_Loop
 bOverrideConcurrency=true
 maxDistance=12500
 soundClassObject=&Fort_Gliders_Land
 virtualizationMode=PlayWhenSilent
 volumeMultiplier=2
 concurrencyOverrides {bLimitToOwner=true, maxCount=2, resolutionRule=StopOldest}
 concurrencySet [&Athena_GliderOpenClose_Conc]
}
